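# MAT_B is a Default Lit, Nanite-used material from an Interchange import, stored with the SparksMaps Jam environment meshes.
# It converts constant Phong ambient, diffuse, specular and shininess inputs through MF_PhongToMetalRoughness, giving a muted plum-pink surface colour with no colour texture.
# A NormalMapWeight parameter blends from a flat normal to the NormalMap texture parameter, defaulting to the full map.
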Material /SparksMaps/Environment/Jam/Meshes/MAT_B
graph {
 properties {assetImportData="<self>:InterchangeAssetImportData_0", bUsedWithNanite=true, usedShadingModels=Default Lit}
 nodes [
  LinearInterpolate(
   A=Constant3Vector() {constant=<rgba 0 0 1 1>}
   B=TextureSampleParameter2D() {parameterName=NormalMap}.RGB
   Alpha=ScalarParameter() {defaultValue=1, parameterName=NormalMapWeight}
  )
  MaterialFunctionCall!(
   AmbientColor=Constant3Vector() {constant=<rgba 0 0 0 1>}
   DiffuseColor=Constant3Vector() {constant=<rgba 0.2247 0.0788 0.1397 1>}
   SpecularColor=Constant3Vector() {constant=<rgba 0.0562 0.0197 0.0349 1>}
   Shininess=Constant() {r=25}
  ) {materialFunction=/InterchangeAssets/Functions/MF_PhongToMetalRoughness}
 ]
}
